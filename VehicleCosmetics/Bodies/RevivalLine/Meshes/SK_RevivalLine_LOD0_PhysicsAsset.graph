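# Physics asset for the RevivalLine vehicle body skeletal mesh, giving its chassis, suspension, steering and wheel bones simple collision bodies.
# Each bone has one capsule, with a large one for the body and a tiny one at the frame.
# Constraints link the body, front differential and rear axle pivots to the frame, with control arms, shocks, wheels and rear brake pads branching from these.
# All joints share uniform 45-degree swing and twist limit angles.
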PhysicsAsset /VehicleCosmetics/Bodies/RevivalLine/Meshes/SK_RevivalLine_LOD0_PhysicsAsset
bodies {
 axle_pivot_bk_l [{shapeType=Capsule, length=21.506, radius=8.8361, center=<-0.7836 -12.8279 -0.2158>, rotation=<rot 0 -0.0269 -88.9561>}]
 axle_pivot_bk_r [{shapeType=Capsule, length=21.4983, radius=8.8356, center=<-0.859 12.8263 -0.1842>, rotation=<rot 0 179.8176 -89.044>}]
 body [{shapeType=Capsule, length=237.4792, radius=95.3873, center=<28.9348 0.0192 22.5984>, rotation=<rot 0 -89.9918 -87.5884>}]
 brakepad_bk_l [{shapeType=Capsule, length=18.6411, radius=16.6345, center=<4.2801 3.7799 0.0829>, rotation=<rot 0 -104.1719 -89.0393>}]
 brakepad_bk_r [{shapeType=Capsule, length=18.6411, radius=16.6345, center=<4.2801 -3.78 0.0829>, rotation=<rot 0 -75.8282 -89.0392>}]
 ctrlArm_bk_l [{shapeType=Capsule, length=22.6355, radius=14.7251, center=<-0.4974 -11.6846 1.2382>, rotation=<rot 0 90.1409 -89.1384>}]
 ctrlArm_bk_r [{shapeType=Capsule, length=22.6348, radius=14.775, center=<0.4542 -11.66 1.2409>, rotation=<rot 0 -90.3297 -89.194>}]
 ctrlArm_fr_l [{shapeType=Capsule, length=22.6346, radius=14.7747, center=<-0.4207 -11.4667 1.313>, rotation=<rot 0 90.3492 -89.2076>}]
 ctrlArm_fr_r [{shapeType=Capsule, length=22.6347, radius=14.7747, center=<0.4207 -11.4666 1.313>, rotation=<rot 0 -90.3492 -89.2076>}]
 frame [{shapeType=Capsule, length=0.505, radius=0.505, rotation=<rot 0 -90 0>}]
 front_differential [{shapeType=Capsule, length=57.9229, radius=5.2143, center=<-0.7936 0 0.1626>, rotation=<rot 0 180 -90>}]
 shock_base_bk_l [{shapeType=Capsule, length=24.1594, radius=22.6299, center=<-0.5482 1.5181 -1.7117>, rotation=<rot 26.2356 -90 -0.2519>}]
 shock_base_bk_r [{shapeType=Capsule, length=24.1665, radius=22.6294, center=<-0.5496 -1.5318 -1.737>, rotation=<rot -26.321 -90 -0.2628>}]
 shock_base_fr_l [{shapeType=Capsule, length=24.7365, radius=22.6133, center=<-0.572 0.7567 -2.712>, rotation=<rot 30.9876 -90 -1.22>}]
 shock_base_fr_r [{shapeType=Capsule, length=24.7405, radius=22.6131, center=<-0.5724 -0.7609 -2.7291>, rotation=<rot -31.0467 -90 -1.2379>}]
 wheel_disc_bk_l [{shapeType=Capsule, length=20.0926, radius=20.0926, center=<0 -1.4261 0>, rotation=<rot 0 -90 0>}]
 wheel_disc_bk_r [{shapeType=Capsule, length=20.0926, radius=20.0926, center=<0 1.4261 0>, rotation=<rot 0 -90 0>}]
 wheel_disc_fr_l [{shapeType=Capsule, length=20.0926, radius=20.0926, center=<0 -1.3812 0>, rotation=<rot 0 -90 0.0001>}]
 wheel_disc_fr_r [{shapeType=Capsule, length=20.068, radius=20.0675, center=<0 1.3795 0.0005>, rotation=<rot -0.0182 -90 8.3815>}]
 wheel_steering_fr_l [{shapeType=Capsule, length=20.7215, radius=19.8864, center=<-8.3696 1.9346 0.3675>, rotation=<rot 0 109.7757 -71.2228>}]
 wheel_steering_fr_r [{shapeType=Capsule, length=20.7118, radius=19.871, center=<-8.3601 -1.9172 0.3701>, rotation=<rot 0 70.1292 -71.3407>}]
}
constraints [
 {bone1Name=body, bone2Name=frame, swing1LimitDegrees=45, swing2LimitDegrees=45, twistLimitDegrees=45}
 {bone1Name=ctrlArm_fr_l, bone2Name=body, swing1LimitDegrees=45, swing2LimitDegrees=45, twistLimitDegrees=45}
 {bone1Name=ctrlArm_fr_r, bone2Name=body, swing1LimitDegrees=45, swing2LimitDegrees=45, twistLimitDegrees=45}
 {bone1Name=ctrlArm_bk_l, bone2Name=body, swing1LimitDegrees=45, swing2LimitDegrees=45, twistLimitDegrees=45}
 {bone1Name=ctrlArm_bk_r, bone2Name=body, swing1LimitDegrees=45, swing2LimitDegrees=45, twistLimitDegrees=45}
 {bone1Name=front_differential, bone2Name=frame, swing1LimitDegrees=45, swing2LimitDegrees=45, twistLimitDegrees=45}
 {bone1Name=shock_base_fr_l, bone2Name=front_differential, swing1LimitDegrees=45, swing2LimitDegrees=45, twistLimitDegrees=45}
 {bone1Name=wheel_steering_fr_l, bone2Name=front_differential, swing1LimitDegrees=45, swing2LimitDegrees=45, twistLimitDegrees=45}
 {bone1Name=wheel_disc_fr_l, bone2Name=wheel_steering_fr_l, swing1LimitDegrees=45, swing2LimitDegrees=45, twistLimitDegrees=45}
 {bone1Name=shock_base_fr_r, bone2Name=front_differential, swing1LimitDegrees=45, swing2LimitDegrees=45, twistLimitDegrees=45}
 {bone1Name=wheel_steering_fr_r, bone2Name=front_differential, swing1LimitDegrees=45, swing2LimitDegrees=45, twistLimitDegrees=45}
 {bone1Name=wheel_disc_fr_r, bone2Name=wheel_steering_fr_r, swing1LimitDegrees=45, swing2LimitDegrees=45, twistLimitDegrees=45}
 {bone1Name=axle_pivot_bk_l, bone2Name=frame, swing1LimitDegrees=45, swing2LimitDegrees=45, twistLimitDegrees=45}
 {bone1Name=shock_base_bk_l, bone2Name=axle_pivot_bk_l, swing1LimitDegrees=45, swing2LimitDegrees=45, twistLimitDegrees=45}
 {bone1Name=wheel_disc_bk_l, bone2Name=axle_pivot_bk_l, swing1LimitDegrees=45, swing2LimitDegrees=45, twistLimitDegrees=45}
 {bone1Name=brakepad_bk_l, bone2Name=axle_pivot_bk_l, swing1LimitDegrees=45, swing2LimitDegrees=45, twistLimitDegrees=45}
 {bone1Name=axle_pivot_bk_r, bone2Name=frame, swing1LimitDegrees=45, swing2LimitDegrees=45, twistLimitDegrees=45}
 {bone1Name=shock_base_bk_r, bone2Name=axle_pivot_bk_r, swing1LimitDegrees=45, swing2LimitDegrees=45, twistLimitDegrees=45}
 {bone1Name=wheel_disc_bk_r, bone2Name=axle_pivot_bk_r, swing1LimitDegrees=45, swing2LimitDegrees=45, twistLimitDegrees=45}
 {bone1Name=brakepad_bk_r, bone2Name=axle_pivot_bk_r, swing1LimitDegrees=45, swing2LimitDegrees=45, twistLimitDegrees=45}
]
